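# Sound cue for the Creative toy ActionTrigger's player-spotted sound.
# A single wave player outputs the ActionTrigger_PlayerSpotted_01 wave directly, with no randomization or mixing.
# It uses the Creative small-toy attenuation and the creative gadgets sound class, at slightly reduced volume.
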SoundCue /Game/Sounds/Creative/Toys/ActionTrigger/ActionTrigger_PlayerSpotted_Cue
root=WavePlayer_2
nodes [{id=WavePlayer_2, properties {soundWaveAssetPtr=&/ActionTrigger_PlayerSpotted_01}}]
properties {
 attenuationSettings=/Game/Sounds/Creative/Attenuations/Fort_Creative_SmallToy_Att
 maxDistance=3000
 soundClassObject=&Fort_Gadgets_Creative
 volumeMultiplier=0.9
}
